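# Looping indoor ambience Sound Cue for a draughty, medium-high absorption room, routed to the indoor ambient sound class with shared attenuation and a loop-limiting concurrency.
# A quality-level switch plays the quad wave in its first two slots and the stereo wave in its last two.
# Each looping wave feeds two indefinitely looping envelopers and a modulator, layering slow pitch and volume drifts of different periods to vary the loop.
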SoundCue /WildEstate/Assets/Audio/POI/Domestic_Commercial/SC_Ambience_Medium_High_Absorption_Draughty_Room_Loop_Int
root=QualityLevel_2
nodes [
 {
  id=Enveloper_4
  children [WavePlayer_3]
  properties {
   bLoop=true
   bLoopIndefinitely=true
   loopEnd=30
   pitchCurve {editorCurveData {keys [
      {arriveTangent=-0.0013, interpMode=RCIM_Cubic, leaveTangent=-0.0013, time=0}
      <at 15 0.98 RCIM_Cubic>
      {arriveTangent=0.0013, interpMode=RCIM_Cubic, leaveTangent=0.0013, time=30}
     ]}}
   volumeCurve {editorCurveData {keys [<at 0 1>]}}
  }
 }
 {
  id=Enveloper_5
  children [Enveloper_4]
  properties {
   bLoop=true
   bLoopIndefinitely=true
   loopEnd=20
   pitchCurve {editorCurveData {keys [{time=0}]}}
   volumeCurve {editorCurveData {keys [
      {arriveTangent=-0.01, interpMode=RCIM_Cubic, time=0, value=1}
      <at 10 0.9 RCIM_Cubic>
      {interpMode=RCIM_Cubic, leaveTangent=0.01, time=20, value=1}
     ]}}
  }
 }
 {
  id=Enveloper_6
  children [WavePlayer_2]
  properties {
   bLoop=true
   bLoopIndefinitely=true
   loopEnd=30
   pitchCurve {editorCurveData {keys [
      {arriveTangent=-0.0013, interpMode=RCIM_Cubic, leaveTangent=-0.0013, time=0}
      <at 15 0.98 RCIM_Cubic>
      {arriveTangent=0.0013, interpMode=RCIM_Cubic, leaveTangent=0.0013, time=30}
     ]}}
   volumeCurve {editorCurveData {keys [<at 0 1>]}}
  }
 }
 {
  id=Enveloper_7
  children [Enveloper_6]
  properties {
   bLoop=true
   bLoopIndefinitely=true
   loopEnd=20
   pitchCurve {editorCurveData {keys [{time=0}]}}
   volumeCurve {editorCurveData {keys [
      {arriveTangent=-0.01, interpMode=RCIM_Cubic, time=0, value=1}
      <at 10 0.9 RCIM_Cubic>
      {interpMode=RCIM_Cubic, leaveTangent=0.01, time=20, value=1}
     ]}}
  }
 }
 {id=Modulator_0, children [Enveloper_7]}
 {id=Modulator_1, children [Enveloper_5]}
 {id=QualityLevel_2, children [Modulator_0,Modulator_0,Modulator_1,Modulator_1]}
 {
  id=WavePlayer_2
  properties {
   bLooping=true
   soundWaveAssetPtr=&/Ambience_Medium_High_Absorption_Draughty_Room_Loop_Int_Quad
  }
 }
 {
  id=WavePlayer_3
  properties {
   bLooping=true
   soundWaveAssetPtr=&/Ambience_Medium_High_Absorption_Draughty_Room_Loop_Int_Stereo
  }
 }
]
properties {
 attenuationSettings=/AmbientAudioBase/Attenuations/Fort_Ambient_Att
 bEnableBusSends=false
 bEnableSubmixSends=false
 maxDistance=2097152
 priority=5
 soundClassObject=&Fort_Ambient_Indoors_Athena
 volumeMultiplier=1
 concurrencySet [/AmbientAudioBase/Concurrency/AmbientConc_2Max_Loop_PreventNew_Default_Concurrency]
}
